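# Post-process style AnimBP for the Good Henchman shorts body, copying the attached parent's pose and adding coat, shorts and finger secondary motion.
# bRigidBodyEnabled blends between RigidBody simulation (borrowing the Bad Henchman physics asset) and a Control Rig fallback, each followed by coat bone offsets, mostly thigh-alpha driven.
# Calf alphas then drive corrective offsets on shorts bones, and rotation constraints make const_middle finger bones follow the middle fingers.
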AnimBlueprint /Game/Characters/Player/Male/Medium/Bodies/M_MED_Henchman_Good_Shorts/Meshes/M_MED_Henchman_Good_Shorts_AnimBP
target_skeleton=&/M_MED_Henchman_Good_Shorts_Skeleton
graphs [
 AnimGraph {nodes [
   n1 = RigidBody!(
    ComponentPose=Local To Component(LocalPose=Use cached pose 'PreCache'())
    Alpha
    OverrideWorldGravity:Vector=Get GravityOverride()
    ExternalForce
   ) -> (Pose) {
    bClampLinearTranslationLimitToRefPose=true
    bOverrideWorldGravity=true
    overridePhysicsAsset=/Game/Characters/Player/Male/Medium/Bodies/M_MED_Henchman_Bad/Meshes/M_MED_Henchman_Bad2_Physics
    simulationSpace=BaseBoneSpace
    alphaBoolBlend {blendOption=Linear}
    baseBoneRef {boneName=spine_01}
    componentAppliedLinearAccClamp=<3000 3000 3000>
    componentLinearAccScale=<0.1 0.05 0.05>
    componentLinearVelScale=<0.1 0.05 0.05>
    overrideWorldGravity {z=-980}
   }
   n2 = Get ThighAlphaRight() -> (ThighAlphaRight:float)
   n3 = "Transform (Modify) Bone - Bone: dyn_coat_MSTR_frnt_r"(ComponentPose=@n1, Alpha=@n2, Translation="-5,0,4", Rotation="0,0,0", Scale="0.9,0.9,0.9") {
    rotationMode=BMM_Additive
    scaleMode=BMM_Replace
    translationMode=BMM_Additive
    alphaBoolBlend {blendInTime=0.5, blendOption=Linear, blendOutTime=1}
    alphaScaleBiasClamp {bInterpResult=true, interpSpeedDecreasing=15, interpSpeedIncreasing=15}
    boneToModify {boneName=dyn_coat_MSTR_frnt_r}
    scale=<0.9 0.9 0.9>
    translation {x=-5, z=4}
   }
   n4 = "Transform (Modify) Bone - Bone: dyn_coat_MSTR_side_r"(ComponentPose=@n3, Alpha=@n2, Translation="-0.5,0,1", Rotation="0,0,0", Scale="1,1,1") {
    rotationMode=BMM_Additive
    translationMode=BMM_Additive
    alphaBoolBlend {blendInTime=0.5, blendOption=Linear, blendOutTime=1}
    alphaScaleBiasClamp {bInterpResult=true, interpSpeedDecreasing=15, interpSpeedIncreasing=15}
    boneToModify {boneName=dyn_coat_MSTR_side_r}
    translation {x=-0.5, z=1}
   }
   n5 = Get ThighAlphaLeft() -> (ThighAlphaLeft:float)
   n6 = "Transform (Modify) Bone - Bone: dyn_coat_MSTR_frnt_l"(ComponentPose=@n4, Alpha=@n5, Translation="4,-2,1", Rotation="0,0,-7", Scale="0.6,0.6,0.6") {
    rotationMode=BMM_Additive
    scaleMode=BMM_Replace
    translationMode=BMM_Additive
    alphaBoolBlend {blendOption=Linear}
    boneToModify {boneName=dyn_coat_MSTR_frnt_l}
    rotation {roll=-7}
    scale=<0.6 0.6 0.6>
    translation=<4 -2 1>
   }
   n7 = "Transform (Modify) Bone - Bone: dyn_coat_MSTR_side_l"(ComponentPose=@n6, Alpha=@n5, Translation="1,0.3,1", Rotation="0,0,0", Scale="0.5,0.5,0.5") {
    rotationMode=BMM_Additive
    scaleMode=BMM_Replace
    translationMode=BMM_Additive
    alphaBoolBlend {blendOption=Linear}
    boneToModify {boneName=dyn_coat_MSTR_side_l}
    scale=<0.5 0.5 0.5>
    translation=<1 0.3 1>
   }
   n8 = "Transform (Modify) Bone - Bone: dyn_coat_MSTR_back_l"(ComponentPose=@n7, Alpha=1, Translation="1.5,-2,0", Rotation="0,0,0", Scale="1,1,1") {
    rotationMode=BMM_Additive
    translationMode=BMM_Additive
    alphaBoolBlend {blendInTime=0.1, blendOption=Linear, blendOutTime=0.1}
    boneToModify {boneName=dyn_coat_MSTR_back_l}
    translation=<1.5 -2>
   }
   n9 = "Transform (Modify) Bone - Bone: dyn_coat_MSTR_back_r"(ComponentPose=@n8, Alpha=1, Translation="0,-2,0", Rotation="0,0,0", Scale="1,1,1") {
    translationMode=BMM_Additive
    alphaBoolBlend {blendInTime=0.1, blendOption=Linear, blendOutTime=0.1}
    boneToModify {boneName=dyn_coat_MSTR_back_r}
    translation {y=-2}
   }
   n10 = "Transform (Modify) Bone - Bone: dyn_coat_MSTR_frnt_l"(ComponentPose=@n9, Alpha=1, Translation="0,2,0", Rotation="0,0,0", Scale="1,1,1") {
    translationMode=BMM_Additive
    alphaBoolBlend {blendInTime=0.1, blendOption=Linear, blendOutTime=0.1}
    boneToModify {boneName=dyn_coat_MSTR_frnt_l}
    translation {y=2}
   }
   n11 = Component To Local(ComponentPose=@n10)
   n12 = Control Rig(Source=Use cached pose 'PreCache'()) {
    alphaBoolBlend {blendOption=Linear}
    controlRigAssetReference {blueprintRigClass=&/Henchman_Shorts_Good_CtrlRig.Henchman_Shorts_Good_CtrlRig_C}
   }
   n13 = "Transform (Modify) Bone - Bone: dyn_coat_MSTR_frnt_l"(ComponentPose=Local To Component(LocalPose=@n12), Alpha=@n5, Translation="0,0,0", Rotation="0,0,-10", Scale="1,1,1") {
    rotationMode=BMM_Additive
    translationMode=BMM_Additive
    alphaBoolBlend {blendInTime=0.1, blendOption=Linear, blendOutTime=0.1}
    boneToModify {boneName=dyn_coat_MSTR_frnt_l}
    rotation {roll=-10}
   }
   n14 = "Transform (Modify) Bone - Bone: dyn_coat_MSTR_side_l"(ComponentPose=@n13, Alpha=@n5, Translation="2,0,0", Rotation="25,0,-80", Scale="1,1,1") {
    rotationMode=BMM_Additive
    translationMode=BMM_Additive
    alphaBoolBlend {blendInTime=0.1, blendOption=Linear, blendOutTime=0.1}
    boneToModify {boneName=dyn_coat_MSTR_side_l}
    rotation {pitch=25, roll=-80}
    translation {x=2}
   }
   n15 = "Transform (Modify) Bone - Bone: dyn_coat_MSTR_side_r"(ComponentPose=@n14, Alpha=1, Translation="0,0,0", Rotation="-5,0.000,0.000", Scale="1,1,1") {
    rotationMode=BMM_Additive
    translationMode=BMM_Additive
    alphaBoolBlend {blendInTime=0.1, blendOption=Linear, blendOutTime=0.1}
    boneToModify {boneName=dyn_coat_MSTR_side_r}
    rotation {pitch=-5}
   }
   n16 = "Transform (Modify) Bone - Bone: dyn_coat_MSTR_frnt_r"(ComponentPose=@n15, Alpha=@n2, Translation="0,0,0", Rotation="-30,0,-20", Scale="1,1,1") {
    rotationMode=BMM_Additive
    translationMode=BMM_Additive
    alphaBoolBlend {blendInTime=0.1, blendOption=Linear, blendOutTime=0.1}
    boneToModify {boneName=dyn_coat_MSTR_frnt_r}
    rotation {pitch=-30, roll=-20}
   }
   n17 = "Transform (Modify) Bone - Bone: dyn_coat_MSTR_side_r"(ComponentPose=@n16, Alpha=@n2, Translation="0,0,0", Rotation="-17,0,0", Scale="1,1,1") {
    rotationMode=BMM_Additive
    translationMode=BMM_Additive
    alphaBoolBlend {blendInTime=0.1, blendOption=Linear, blendOutTime=0.1}
    boneToModify {boneName=dyn_coat_MSTR_side_r}
    rotation {pitch=-17}
   }
   n18 = Component To Local(ComponentPose=@n17)
   n19 = Get bRigidBodyEnabled()
   n20 = Get CalfAlphaLeft() -> (CalfAlphaLeft:float)
   n21 = "Transform (Modify) Bone - Bone: Dyn_short_back_l"(
    ComponentPose=Local To Component(LocalPose=Blend Poses by bool(BlendPose_0=@n11, BlendPose_1=@n18, bActiveValue=@n19) {blendType=Linear})
    Alpha=@n20
    Translation="-12,0,2"
    Rotation="-3,0,0"
    Scale="1,1,1"
   ) {
    rotationMode=BMM_Additive
    rotationSpace=BCS_BoneSpace
    scaleMode=BMM_Additive
    scaleSpace=BCS_BoneSpace
    translationMode=BMM_Additive
    translationSpace=BCS_BoneSpace
    alphaBoolBlend {blendOption=Linear}
    alphaScaleBiasClamp {bInterpResult=true, interpSpeedDecreasing=100, interpSpeedIncreasing=100}
    boneToModify {boneName=Dyn_short_back_l}
    rotation {pitch=-3}
    translation {x=-12, z=2}
   }
   n22 = "Transform (Modify) Bone - Bone: Dyn_short_outer_l"(ComponentPose=@n21, Alpha=@n20, Translation="-4,0,-2", Rotation="8,0.000,0", Scale="1,1,1") {
    rotationMode=BMM_Additive
    rotationSpace=BCS_BoneSpace
    scaleMode=BMM_Additive
    scaleSpace=BCS_BoneSpace
    translationMode=BMM_Additive
    translationSpace=BCS_BoneSpace
    alphaBoolBlend {blendOption=Linear}
    alphaScaleBiasClamp {bInterpResult=true, interpSpeedDecreasing=100, interpSpeedIncreasing=100}
    boneToModify {boneName=Dyn_short_outer_l}
    rotation {pitch=8}
    translation {x=-4, z=-2}
   }
   n23 = "Transform (Modify) Bone - Bone: Dyn_short_inner_l"(ComponentPose=@n22, Alpha=@n20, Translation="0,0,2", Rotation="0,0,0", Scale="1,1,1") {
    rotationMode=BMM_Additive
    rotationSpace=BCS_BoneSpace
    scaleMode=BMM_Additive
    scaleSpace=BCS_BoneSpace
    translationMode=BMM_Additive
    translationSpace=BCS_BoneSpace
    alphaBoolBlend {blendOption=Linear}
    alphaScaleBiasClamp {bInterpResult=true, interpSpeedDecreasing=100, interpSpeedIncreasing=100}
    boneToModify {boneName=Dyn_short_inner_l}
    translation {z=2}
   }
   n24 = "Transform (Modify) Bone - Bone: Dyn_short_front_l"(ComponentPose=@n23, Alpha=@n20, Translation="0,0,2.5", Rotation="-13,0,0", Scale="1,1,1") {
    rotationMode=BMM_Additive
    rotationSpace=BCS_BoneSpace
    scaleMode=BMM_Additive
    scaleSpace=BCS_BoneSpace
    translationMode=BMM_Additive
    translationSpace=BCS_BoneSpace
    alphaBoolBlend {blendOption=Linear}
    alphaScaleBiasClamp {bInterpResult=true, interpSpeedDecreasing=100, interpSpeedIncreasing=100}
    boneToModify {boneName=Dyn_short_front_l}
    rotation {pitch=-13}
    translation {z=2.5}
   }
   n25 = Get CalfAlphaRight() -> (CalfAlphaRight:float)
   n26 = "Transform (Modify) Bone - Bone: Dyn_short_back_r"(ComponentPose=@n24, Alpha=@n25, Translation="-12,0,3", Rotation="-3,0,0", Scale="0.9,0.9,0.9") {
    rotationMode=BMM_Additive
    rotationSpace=BCS_BoneSpace
    scaleMode=BMM_Additive
    scaleSpace=BCS_BoneSpace
    translationMode=BMM_Additive
    translationSpace=BCS_BoneSpace
    alphaBoolBlend {blendOption=Linear}
    alphaScaleBiasClamp {bInterpResult=true, interpSpeedDecreasing=100, interpSpeedIncreasing=100}
    boneToModify {boneName=Dyn_short_back_r}
    rotation {pitch=-3}
    scale=<0.9 0.9 0.9>
    translation {x=-12, z=3}
   }
   n27 = "Transform (Modify) Bone - Bone: Dyn_short_outer_r"(ComponentPose=@n26, Alpha=@n25, Translation="-4,0,-3", Rotation="8,0.000,0", Scale="1,1,1") {
    rotationMode=BMM_Additive
    rotationSpace=BCS_BoneSpace
    scaleMode=BMM_Additive
    scaleSpace=BCS_BoneSpace
    translationMode=BMM_Additive
    translationSpace=BCS_BoneSpace
    alphaBoolBlend {blendOption=Linear}
    alphaScaleBiasClamp {bInterpResult=true, interpSpeedDecreasing=100, interpSpeedIncreasing=100}
    boneToModify {boneName=Dyn_short_outer_r}
    rotation {pitch=8}
    translation {x=-4, z=-3}
   }
   n28 = "Transform (Modify) Bone - Bone: Dyn_short_inner_r"(ComponentPose=@n27, Alpha=@n25, Translation="-3,-1,3.000", Rotation="0,0,0", Scale="1,1,1") {
    rotationMode=BMM_Additive
    rotationSpace=BCS_BoneSpace
    scaleMode=BMM_Additive
    scaleSpace=BCS_BoneSpace
    translationMode=BMM_Additive
    translationSpace=BCS_BoneSpace
    alphaBoolBlend {blendOption=Linear}
    alphaScaleBiasClamp {bInterpResult=true, interpSpeedDecreasing=100, interpSpeedIncreasing=100}
    boneToModify {boneName=Dyn_short_inner_r}
    translation=<-3 -1 3>
   }
   n29 = "Transform (Modify) Bone - Bone: Dyn_short_front_r"(ComponentPose=@n28, Alpha=@n25, Translation="0,0,2.5", Rotation="-8,0,0", Scale="1,1,1") {
    rotationMode=BMM_Additive
    rotationSpace=BCS_BoneSpace
    scaleMode=BMM_Additive
    scaleSpace=BCS_BoneSpace
    translationMode=BMM_Additive
    translationSpace=BCS_BoneSpace
    alphaBoolBlend {blendOption=Linear}
    alphaScaleBiasClamp {bInterpResult=true, interpSpeedDecreasing=100, interpSpeedIncreasing=100}
    boneToModify {boneName=Dyn_short_front_r}
    rotation {pitch=-8}
    translation {z=2.5}
   }
   n30 = "Constraint - const_middle_01_l (R)"(ComponentPose=@n29, Alpha=1, ConstraintWeights_0=1) {
    alphaBoolBlend {blendOption=Linear}
    boneToModify {boneName=const_middle_01_l}
    constraintSetup [{transformType=Rotation, targetBone {boneName=middle_01_l}}]
    constraintWeights [1]
   }
   n31 = "Constraint - const_middle_02_l (R)"(ComponentPose=@n30, Alpha=0.5, ConstraintWeights_0=1) {
    alphaBoolBlend {blendOption=Linear}
    boneToModify {boneName=const_middle_02_l}
    constraintSetup [{transformType=Rotation, targetBone {boneName=middle_02_l}}]
    constraintWeights [1]
   }
   n32 = "Constraint - const_middle_03_l (R)"(ComponentPose=@n31, Alpha=0.5, ConstraintWeights_0=1) {
    alphaBoolBlend {blendOption=Linear}
    boneToModify {boneName=const_middle_03_l}
    constraintSetup [{transformType=Rotation, targetBone {boneName=middle_03_l}}]
    constraintWeights [1]
   }
   n33 = "Constraint - const_middle_01_r (R)"(ComponentPose=@n32, Alpha=1, ConstraintWeights_0=1) {
    alphaBoolBlend {blendOption=Linear}
    boneToModify {boneName=const_middle_01_r}
    constraintSetup [{transformType=Rotation, targetBone {boneName=middle_01_r}}]
    constraintWeights [1]
   }
   n34 = "Constraint - const_middle_02_r (R)"(ComponentPose=@n33, Alpha=0.5, ConstraintWeights_0=1) {
    alphaBoolBlend {blendOption=Linear}
    boneToModify {boneName=const_middle_02_r}
    constraintSetup [{transformType=Rotation, targetBone {boneName=middle_02_r}}]
    constraintWeights [1]
   }
   n35 = "Constraint - const_middle_03_r (R)"(ComponentPose=@n34, Alpha=0.5, ConstraintWeights_0=1) {
    alphaBoolBlend {blendOption=Linear}
    boneToModify {boneName=const_middle_03_r}
    constraintSetup [{transformType=Rotation, targetBone {boneName=middle_03_r}}]
    constraintWeights [1]
   }
   Output Pose(Result=Component To Local(ComponentPose=@n35)) {layerGroup=None}
   Save cached pose 'PreCache'(Pose=Copy Pose From Mesh() {bUseAttachedParent=true}) {cacheName=PreCache}
   Break Fort Anim Input Skydiving External Force(FortAnimInput_SkydivingExternalForce=Make Fort Anim Input Skydiving External Force!() -> (FortAnimInput_SkydivingExternalForce:Fort Anim Input Skydiving External Force Structure) {is=title K2Node_MakeStruct}) -> (
    bUseSkydivingVectorForce=~
    SkydivingVectorForce=~
    HeadToPelvisDirection=~
    FloatingMultiplier=~
    FloatingAdditive=~
    DivingMultiplier=~
    DivingAdditive=~
    ParachutingMultiplier=~
    ParachutingAdditive=~
   )
   Get SkydivingExternalForceInput()
  ]}
 EventGraph {links [], nodes []}
]
variables []
